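# Unlit, two-sided translucent Niagara ribbon material for the Eliminationv2 character ResIn effect.
# Three sharpened sine bands along the V coordinate, shifted apart by ChromaOffset, form RGB channels for a chromatic fringe tinted by particle color.
# The color passes through DeriveHDRfromLDR and MF_FX_EmissiveAdjustments for emissive.
# A separate powered sine on V, multiplied by particle alpha, gives a soft-edged alpha falloff.
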Material /FNGameplayCues/Effects/Fort_Effects/Effects/Characters/Generic/ResIn/Eliminationv2/M_Eliminationv2_Ribbon
graph {
 properties {
  bUsedWithNiagaraRibbons=true
  blendMode=BLEND_Translucent
  pixelDepthOffsetMode=PDOM_Legacy
  shadingModel=MSM_Unlit
  twoSided=true
  usedShadingModels=Unlit
 }
 nodes [
  n1 = ParticleColor()
  n2 = ComponentMask(None=TextureCoordinate()) {g=true}
  n3 = ScalarParameter() {defaultValue=0.05, parameterName=ChromaOffset}
  n4 = Sine(None=Add(A=@n2, B=@n3)) {period=2}
  n5 = ScalarParameter() {defaultValue=4, parameterName=Param2}
  n6 = Sine(None=@n2) {period=2}
  n7 = ScalarParameter() {defaultValue=4, parameterName=Param2}
  n8 = Sine(None=Add(A=@n2, B=Multiply(A=@n3) {constB=-1})) {period=2}
  n9 = ScalarParameter() {defaultValue=4, parameterName=Param2}
  n10 = Multiply(
   A=@n1.RGB
   B=MaterialFunctionCall(X=Power(Base=@n4, Exp=@n5), Y=Power(Base=@n6, Exp=@n7), Z=Power(Base=@n8, Exp=@n9)) {materialFunction=&MakeFloat3}.Result
  )
  MaterialFunctionCall(Emissive=MaterialFunctionCall(LDR Input=Clamp(None=@n10), HDR Tint=@n10) {materialFunction=&DeriveHDRfromLDR}.RemappedRange) {materialFunction=&MF_FX_EmissiveAdjustments}
  n11 = ComponentMask(None=TextureCoordinate()) {g=true}
  Multiply(A=@n1.A, B=Power(Base=Sine(None=@n11) {period=2}, Exp=ScalarParameter() {defaultValue=2}))
 ]
}
